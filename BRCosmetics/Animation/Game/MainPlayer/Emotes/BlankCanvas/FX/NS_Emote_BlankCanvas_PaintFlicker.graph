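# Niagara system for the paint flicker effect of the BlankCanvas player emote.
# Its three minimal emitters list no renderers, so none draws anything itself.
# The system stack uses two variable-setting modules in system spawn plus SystemState, with dynamic inputs for random floats, float comparison to bool, and bool or float conversions.
# It exposes User.Bool, which is typed as a float despite its name.
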NiagaraSystem /BRCosmetics/Animation/Game/MainPlayer/Emotes/BlankCanvas/FX/NS_Emote_BlankCanvas_PaintFlicker
dependencies {
 usedDataInterfaces []
 usedDynamicInputs [
  &MakeIntFromFloat
  &MakeCustomFloatFromBool
  &SetBoolByFloatComparison
  &RandomRangeFloat
 ]
 usedModules [
  &self:SystemSpawnScript.SystemScriptSource.SystemScriptGraph.NiagaraNodeAssignment_0.SetVariables_#1
  &self:SystemSpawnScript.SystemScriptSource.SystemScriptGraph.NiagaraNodeAssignment_1.SetVariables_#2
  &SystemState
 ]
}
emitters {Minimal {rendererClasses [], renderers []}, Minimal001 {rendererClasses [], renderers []}, Minimal002 {rendererClasses [], renderers []}}
summary {
 systemName=NS_Emote_BlankCanvas_PaintFlicker
 emitters [{emitterName=Minimal, rendererClasses []}, {emitterName=Minimal001, rendererClasses []}, {emitterName=Minimal002, rendererClasses []}]
 userVariables [User.Bool {type=NiagaraFloat, defaultValue=1, description=""}]
}
